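; Creative spawner device for the Baller (Octopus Vehicle V2) that maps energy options to vehicle fuel.
; It applies device options to data registries under per-instance resolver scopes, then refreshes the vehicle's fuel usage and towhook range.
; UpdateVehicleConfigs caches the vehicle and sets starting and infinite energy before calling the parent.
; A refill event fills the fuel, and the out-of-fuel delegate broadcasts an out-of-energy gameplay event with the driver's controller.
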(event EventOnVehicleSpawned
  (params (Spawned%20Vehicle :type Fort%20Athena%20Vehicle%20Object%20Reference))
  (|CallParentFunction Spawned%20Vehicle)
  (EventDispatchers|BindEventtoOnOutOfFuelDelegate Spawned%20Vehicle (EventDispatchers|CreateEvent :!event Custom|OutOfFuel)))

(event Custom|On%20Refill%20Energy
  (params (Instigator :type Controller%20Object%20Reference))
  (bind _octopus_vehicle_v2 (Variables|Default|GetOctopus%20Vehicle%20V2)
    (:then
      (Vehicle|Fuel|SetFuel _octopus_vehicle_v2 (Vehicle|Fuel|GetFuelCapacity _octopus_vehicle_v2)))
    (:else)))

(event Custom|OutOfFuel
  (bind _octopus_vehicle_v2 (Variables|Default|GetOctopus%20Vehicle%20V2)
    (:then
      (GameplayEvent|BroadcastEvent (Variables|UserOptions|Events|GetOnIsOutOfEnergy) 0 (Utilities|Struct|MakeGameplayEventVehicleOutOfEnergy (Pawn|GetController (Vehicle|GetDriver _octopus_vehicle_v2)))))
    (:else)))

(fn BuildDataRegistryResolverScope
  (return (Variables|Default|GetResolver%20Scopes) 1 true))

(fn GetCacheContextOverride
  (return 0))

(fn InitializeDevice
  (Class|FortAthenaMutatorUtilities|RemoveDataRegistryOptionsforContextObject)
  (Class|FortAthenaMutatorUtilities|ApplyActorOptionsComponenttoDataRegistry (Variables|Default|GetToyOptionsComponent))
  (|CallParentFunction))

(fn UpdateVehicleConfigs
  (Utilities|FlowControl|Sequence
    (:then_0
      (bind _octopus_vehicle_v2 (Variables|Default|GetOctopus%20Vehicle%20V2)
        (:then
          (EventDispatchers|UnbindEventfromOnOutOfFuelDelegate _octopus_vehicle_v2 (EventDispatchers|CreateEvent :!event OutOfFuel)))
        (:else
          _))
      (bind _asoctopus_vehicle_v2 (Utilities|Casting|CastToOctopusVehicle_V2 (Variables|Default|GetCachedVehicleInterface)))
      (bind _output_get (Variables|Default|SetOctopus%20Vehicle%20V2 _asoctopus_vehicle_v2))
      ;[1 1 1] Refresh DataRegistry values now that vehicle is in the correct registry scope
      (Class|FuelComponent|ReinitializeFuelUsageInfo (Vehicle|Fuel|GetVehicleFuelComponent _output_get))
      (Class|OctopusVehicleV2|RefreshTowhookRange _output_get)
      (Variables|UserKnobs|SetStartingFuel (Variables|UserOptions|GetStartingEnergy))
      (Variables|Default|SetbHasInfiniteFuel (!= (Variables|UserOptions|GetHasInfiniteEnergy) 0)) ; Needs to be set or will override vehicle setting
    )
    (:then_1
      (|CallParentFunction))))

(fn ConstructionScript
  (|CallParentFunction)
  (Variables|Default|SetResolver%20Scopes (Utilities|Array|MakeArray (SpatialGameplay|GetSpatialGameplayInterfaceFNameforActor self) (Utilities|String|StringToName (Utilities|GetDisplayName self)))))
