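; Baller vehicle spawner device that applies user options for energy, grappler range and wrap support to the Octopus vehicle it spawns.
; Getter overrides feed options to the parent spawner, while UpdateVehicleConfigs caches the vehicle, reinitializes fuel data, overrides towhook range and optionally randomizes fuel.
; An out-of-fuel delegate broadcasts a gameplay event with the driver's controller, and a refill handler restores full fuel.
; Legacy migration converts old starting energy, treating -10 as random.
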(event EventOnVehicleSpawned
  (params (Spawned%20Vehicle :type Fort%20Athena%20Vehicle%20Object%20Reference))
  (|CallParentFunction Spawned%20Vehicle)
  (EventDispatchers|BindEventtoOnOutOfFuelDelegate Spawned%20Vehicle (EventDispatchers|CreateEvent :!event Custom|OutOfFuel)))

(event Custom|On%20Refill%20Energy
  (params (Instigator :type Controller%20Object%20Reference))
  (bind _octopus_vehicle_v2 (Variables|Default|GetOctopus%20Vehicle%20V2)
    (:then
      (Vehicle|Fuel|SetFuel _octopus_vehicle_v2 (Vehicle|Fuel|GetFuelCapacity _octopus_vehicle_v2)))
    (:else)))

(event Custom|OutOfFuel
  (bind _octopus_vehicle_v2 (Variables|Default|GetOctopus%20Vehicle%20V2)
    (:then
      (GameplayEvent|BroadcastEvent (Variables|UserOptions|Events|GetOnIsOutOfEnergy) 0 (Utilities|Struct|MakeGameplayEventVehicleOutOfEnergy (Pawn|GetController (Vehicle|GetDriver _octopus_vehicle_v2)))))
    (:else)))

(fn GetSupportsWraps
  (return (Variables|UserOptions|Basic|GetSupportsWraps)))

(fn GetConsumesFuel
  (return (not (Variables|UserOptions|GetHasInfiniteEnergy))))

(fn GetRandomStartingFuel
  (return (Variables|UserOptions|GetUseRandomStartingEnergy)))

(fn GetStartingFuelValue
  (return (Math|Integer|SelectInt -10 (Variables|UserOptions|GetStarting%20Energy) (Variables|UserOptions|GetUseRandomStartingEnergy))))

(fn InitializeDevice
  (|CallParentFunction))

(fn MigrateLegacyUserOptions
  (bind _startingenergy (Class|GetClassDefaults "/CRD_VehicleSpawners/SetupAssets/Baller/Device_BallerSpawner_V2.Device_BallerSpawner_V2_C"))
  (|CallParentFunction)
  (Utilities|FlowControl|Sequence
    (:then_0
      (Utilities|FlowControl|Sequence
        (:then_0
          (if (!= (Variables|Legacy|GetStartingEnergy) _startingenergy)
            (if (== (Variables|Legacy|GetStartingEnergy) -10)
              (Variables|UserOptions|SetUseRandomStartingEnergy true)
              (else
                (Variables|UserOptions|SetStarting%20Energy (Variables|Legacy|GetStartingEnergy))))
            (Variables|Legacy|SetStartingEnergy _startingenergy)
            (LocalVariables|SetMigratedOption true)))))
    (:then_1
      (if (LocalVariables|GetMigratedOption)
        (OptionData|RefreshPropertyOverrides (Variables|Default|GetToyOptionsComponent))))))

(fn UpdateVehicleConfigs
  (Utilities|FlowControl|Sequence
    (:then_0
      (bind _octopus_vehicle_v2 (Variables|Default|GetOctopus%20Vehicle%20V2)
        (:then
          (EventDispatchers|UnbindEventfromOnOutOfFuelDelegate _octopus_vehicle_v2 (EventDispatchers|CreateEvent :!event OutOfFuel)))
        (:else
          _))
      (bind _asoctopus_vehicle_v2 (Utilities|Casting|CastToOctopusVehicle_V2 (Variables|Default|GetCachedVehicleInterface)))
      (bind _output_get (Variables|Default|SetOctopus%20Vehicle%20V2 _asoctopus_vehicle_v2))
      ;[1 1 1] Refresh DataRegistry values now that vehicle is in the correct registry scope
      (Class|FuelComponent|ReinitializeFuelUsageInfo (Vehicle|Fuel|GetVehicleFuelComponent _output_get))
      (Class|OctopusVehicleV2|OverrideTowhookRange _output_get (* (Variables|UserOptions|GetGrappler%20Range) 1e+02))
      (if (Variables|UserOptions|GetUseRandomStartingEnergy)
        (Vehicle|Fuel|SetFuel _output_get (Math|Random|RandomFloatinRange (* (Class|FuelComponent|GetFuelCapacity (Vehicle|Fuel|GetVehicleFuelComponent _output_get)) 0.25) (* (Class|FuelComponent|GetFuelCapacity (Vehicle|Fuel|GetVehicleFuelComponent _output_get)) 0.8)))))
    (:then_1
      (|CallParentFunction))))

(fn ConstructionScript
  (|CallParentFunction))
